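# Niagara effect for the LemonWrath emote that forms a glowing orange wall with ground cracks, embers and energy sprites.
# Main bursts a long-lived layered mesh with curve-animated material parameters, alongside two crack decals, one with a brief light flash.
# Looping box, energy and flare emitters add particles, using sliding kill volumes or a widening spawn box driven by emitter age.
# System RangeSize scales the wall and flare area; front-end and time-of-day scalars adjust brightness.
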NiagaraSystem /BRCosmetics/Effects/Fort_Effects/Emotes/LemonWrath/NS_Emote_LemonWrath_Wall
dependencies {
 usedDataInterfaces [&NiagaraDataInterfaceCurve, &NiagaraDataInterfaceArrayFloat]
 usedDynamicInputs [&RandomRangeFloat, &SelectFloatFromArray]
 usedModules [
  &self:SystemSpawnScript.SystemScriptSource.SystemScriptGraph.NiagaraNodeAssignment_0.SetVariables_#1
  &self:SystemSpawnScript.SystemScriptSource.SystemScriptGraph.NiagaraNodeAssignment_2.SetVariables_#2
  &SystemState
 ]
}
emitters {
 Decal_crack001 {
  emitterUpdateScript [
   &EmitterState(Inactive Response=::NewEnumerator1, Life Cycle Mode=::NewEnumerator1, Loop Behavior=::NewEnumerator1, Loop Duration=30, UseLoopDelay=false)
   &SpawnBurst_Instantaneous(Use Spawn Probability=false, UseLoopCountLimit=false)
  ]
  particleSpawnScript [&InitializeParticle(Color=<rgba 0.1 0.018 0.003 1>, Lifetime=30, Position Offset=<0 -1 0>, UsePositionOffset=true, Write Material Random=true)]
  particleUpdateScript [
   /Niagara/Modules/Decal/Decal_Attributes(
    Color=1
    Decal Fade:NiagaraFloat=&FloatFromCurve(FloatCurve=DataInterface() {propertyValues {Curve {keys [
         {interpMode=RCIM_Cubic, tangentMode=RCTM_User, time=0, value=0}
         {interpMode=RCIM_Cubic, tangentMode=RCTM_User, time=0.06, value=1}
         {arriveTangent=-0.0482, interpMode=RCIM_Cubic, leaveTangent=-2.9344, tangentMode=RCTM_Break, time=0.2, value=1}
         <at 0.208 0 RCIM_Cubic>
        ]}}})
    Decal Size:Vector3f=<70 110 50>
    Fade=1
    Scale Alpha:NiagaraFloat=&FloatFromCurve(FloatCurve=DataInterface() {propertyValues {Curve {keys [<at 0 1.1>,<at 0.15 -1>,<at 0.9 -1>]}}})
    Scale RGB:Vector3f=&VectorFromFloat(Value=1)
    ScaleA=true
    ScaleRGB=true
    Size=1
   )
   &ParticleState()
   &Light_Attributes(
    Color=1
    Color Mode=0
    Falloff Exponent=2
    Light Color:LinearColor=&Multiply_LinearColorByFloat(
     Linear Color=<rgba 20 6.6667 0 1>
     Float=&Multiply_Float(
      A=&FloatFromCurve(FloatCurve=DataInterface() {propertyValues {Curve {
          defaultValue=0
          keys [
           {interpMode=RCIM_Cubic, leaveTangent=4.5489, tangentMode=RCTM_Break, time=0, value=0}
           {arriveTangent=-0.0854, interpMode=RCIM_Cubic, tangentMode=RCTM_Break, time=0.1032, value=0.9965}
           {arriveTangent=0.019, interpMode=RCIM_Cubic, leaveTangent=0.019, tangentMode=RCTM_Break, time=0.1634, value=0.995}
           {arriveTangent=-0.0655, tangentMode=RCTM_User, time=0.2067, value=0}
          ]
         }}})
      B=1
     )
    )
    Light Exponent=3
    Light Position Offset:Vector3f=<0 0 50>
    Light Radius=120
    Position=1
    Radius=2
    UsePositionOffset=true
   )
  ]
  rendererClasses [&NiagaraDecalRendererProperties, &NiagaraLightRendererProperties]
  renderers [
   {rendererClass=&NiagaraDecalRendererProperties, rendererIndex=0}
   {rendererClass=&NiagaraLightRendererProperties, rendererIndex=1}
  ]
 }
 Decal_cracksmall {
  emitterUpdateScript [
   &EmitterState(Inactive Response=::NewEnumerator1, Life Cycle Mode=::NewEnumerator1, Loop Behavior=::NewEnumerator1, Loop Duration=3.5, UseLoopDelay=true)
   &SpawnBurst_Instantaneous(Use Spawn Probability=false, UseLoopCountLimit=false)
  ]
  particleSpawnScript [&InitializeParticle(Color=<rgba 0.474 0.1007 0 1>, Lifetime=3.5, Position Offset=<45 -3 0>, UsePositionOffset=true, Write Material Random=true)]
  particleUpdateScript [
   &Decal_Attributes(
    Color=1
    Decal Fade:NiagaraFloat=&FloatFromCurve(FloatCurve=DataInterface() {propertyValues {Curve {keys [<at 0.4 1>,<at 0.4982 -0.007>]}}})
    Decal Size:Vector3f=<70 45 40>
    Fade=1
    Scale Alpha:NiagaraFloat=&FloatFromCurve(FloatCurve=DataInterface() {propertyValues {Curve {defaultValue=-1, keys [{tangentMode=RCTM_User, time=0, value=0}, <at 0.6 0>, <at 1 -1>]}}})
    Scale RGB:Vector3f=&VectorFromFloat(Value=&FloatFromCurve(
      FloatCurve=DataInterface() {propertyValues {Curve {defaultValue=1, keys [{tangentMode=RCTM_User, time=0, value=1}, <at 0.6 0>]}}}
      Scale Curve=0.5
     ))
    ScaleA=true
    ScaleRGB=true
    Size=1
   )
   &ParticleState()
  ]
  rendererClasses [&NiagaraDecalRendererProperties]
  renderers [{rendererClass=&NiagaraDecalRendererProperties, rendererIndex=0}]
 }
 Main {
  emitterUpdateScript [
   &EmitterState(Inactive Response=::NewEnumerator1, Life Cycle Mode=::NewEnumerator1, Loop Behavior=::NewEnumerator1, Loop Duration=7, UseLoopDelay=false)
   &SpawnBurst_Instantaneous(Spawn Count=1, Spawn Time=0.1, Use Spawn Probability=false)
  ]
  particleSpawnScript [
   &InitializeParticle(
    Color=<rgba 1 0.2 0 1.5>
    Lifetime=6.8
    Mesh Scale=&Multiply_Vector(A=<1 2.2 1.1>, B=$System.RangeSize)
    Mesh Scale Mode=::NewEnumerator4
    Position Offset=<0 0 -40>
    UsePositionOffset=true
    Write Material Random=true
   ) {module=InitializeParticle001}
   &InitialMeshOrientation(Mesh Orientation Mode=::NewEnumerator2, Rotation=<0 0 -0.25>, Use Rotation Vector=true)
  ]
  particleUpdateScript [
   &ParticleState()
   &DynamicMaterialParameters(
    Index 0 Param 2=1
    Index 0 Param 3=1
    Index 0 Param 4=&FloatFromCurve(FloatCurve=DataInterface() {propertyValues {Curve {
        defaultValue=1
        keys [
         {interpMode=RCIM_Cubic, leaveTangent=0.9684, tangentMode=RCTM_Break, time=0, value=0}
         {arriveTangent=2.4396, interpMode=RCIM_Cubic, leaveTangent=2.4396, tangentMode=RCTM_User, time=0.1351, value=0.3285}
         {arriveTangent=1.3166, interpMode=RCIM_Cubic, leaveTangent=1.3166, time=0.2718, value=0.6348}
         {arriveTangent=1.1227, interpMode=RCIM_Cubic, leaveTangent=1.1227, tangentMode=RCTM_User, time=0.4828, value=0.7862}
         {arriveTangent=1.1517, interpMode=RCIM_Cubic, tangentMode=RCTM_Break, time=0.7, value=1}
        ]
       }}})
    Index 3 Param 1=1
    Index 3 Param 2=&FloatFromCurve(FloatCurve=DataInterface() {propertyValues {Curve {
        defaultValue=0
        keys [{interpMode=RCIM_Cubic, tangentMode=RCTM_User, time=0, value=0}, {interpMode=RCIM_Cubic, tangentMode=RCTM_User, time=1, value=1}]
       }}})
    Module.Index0Float0Write=true
    Module.Index0Float1Write=true
    Module.Index0Float2Write=true
    Module.Index0Float3Write=true
    Module.Index3Float0Write=true
    Module.Index3Float1Write=true
    Module.Index3Float2Write=true
    Module.Index3Float3Write=true
    Write Parameter Index 3=::NewEnumerator1
   )
   &ScaleColor(
    Scale Alpha=&FloatFromCurve(
     FloatCurve=DataInterface() {propertyValues {Curve {keys [<at 0 1>,<at 0.7 1>,<at 0.9 0>]}}}
     Scale Curve=&Multiply_Float(
      A=&FN_FrontEndFloatModifier(Gameplay=1.5, Front End=0.8)
      B=&FNTimeOfDayScalar(Night Scale=0.8, Morning Scale=0.8, Evening Scale=0.8)
     )
    )
    ScaleA=true
    ScaleRGB=true
   )
   &FN_Mesh_Attributes(
    Color=&Multiply_LinearColorByFloat(Linear Color=<rgba 2 0.65 0.05 1>, Float=&FloatFromCurve(FloatCurve=$System.MainAlpha, Scale Curve=2))
    Create Color Attribute=true
    Create Scale Attribute=true
    Module Output=::NewEnumerator1
    Scale=&Multiply_Vector(A=<1 2.5 2>, B=$System.RangeSize)
   )
   &FN_Mesh_Attributes(
    Color=&Multiply_LinearColorByFloat(Linear Color=<rgba 3 1.1146 0.1719 2>, Float=&FloatFromCurve(FloatCurve=$System.MainAlpha))
    Create Color Attribute=true
    Create Scale Attribute=true
    Module Output=::NewEnumerator1
    Scale=&Multiply_Vector(A=<1 2.2 1>, B=$System.RangeSize)
   ) {module=FN_Mesh_Attributes002}
   &FN_Mesh_Attributes(
    Color=&Multiply_LinearColorByFloat(Linear Color=<rgba 1 0.3667 0.05 1>, Float=&FloatFromCurve(FloatCurve=$System.MainAlpha))
    Create Color Attribute=true
    Create Scale Attribute=true
    Module Output=::NewEnumerator1
    Scale=&Multiply_Vector(A=<1 2.5 0.7>, B=$System.RangeSize)
   ) {module=FN_Mesh_Attributes003}
   &Light_Attributes(
    Color=1
    Color Mode=0
    Falloff Exponent=2
    Light Color:LinearColor=&Multiply_LinearColorByFloat(
     Linear Color=<rgba 20 6.6667 0 1>
     Float=&Multiply_Float(
      A=&FloatFromCurve(FloatCurve=DataInterface() {propertyValues {Curve {
          defaultValue=0
          keys [
           {interpMode=RCIM_Cubic, tangentMode=RCTM_Break, time=0.2, value=0}
           {arriveTangent=4.8759, interpMode=RCIM_Cubic, tangentMode=RCTM_Break, time=0.6, value=1}
           {arriveTangent=-8.8925, tangentMode=RCTM_User, time=1, value=0}
          ]
         }}})
      B=1
     )
    )
    Light Exponent=2
    Light Position Offset:Vector3f=<0 -50 60>
    Light Radius=120
    Position=1
    Radius=2
    UsePositionOffset=true
   )
  ]
  rendererClasses [&NiagaraMeshRendererProperties]
  renderers [
   {rendererClass=&NiagaraMeshRendererProperties, rendererIndex=0}
   {rendererClass=&NiagaraMeshRendererProperties, rendererIndex=1}
   {rendererClass=&NiagaraMeshRendererProperties, rendererIndex=2}
   {rendererClass=&NiagaraMeshRendererProperties, rendererIndex=3}
  ]
 }
 box {
  emitterUpdateScript [
   &EmitterState(Life Cycle Mode=::NewEnumerator1, Loop Behavior=::NewEnumerator1, Loop Delay=0.8, Loop Duration=4.6, UseLoopDelay=true)
   &SpawnRate(SpawnRate=40, Use Spawn Probability=false)
  ]
  particleSpawnScript [
   &InitializeParticle(
    Color=<rgba 1 0.25 0.06 1>
    Lifetime Max=0.8
    Lifetime Min=0.4
    Lifetime Mode=::NewEnumerator1
    Mesh Scale Max=<1.1 0.8 0.7>
    Mesh Scale Min=<0.3 0.2 0.3>
    Mesh Scale Mode=::NewEnumerator2
    Position Offset=<130 0 0>
    Sprite Size Max=<1.5 4>
    Sprite Size Min=<1 3>
    Sprite Size Mode=::NewEnumerator2
    UsePositionOffset=true
    Write Material Random=true
   )
   &ShapeLocation(
    Box Size=<90 5 70>
    MaskSpawnGroup=false
    Offset=<-130 0 5>
    Offset Mode=::NewEnumerator0
    Override Seed=false
    Shape Primitive=::NewEnumerator2
    Surface Only=false
    UseOwnerScale=true
   ) {module=ShapeLocation001}
   &AddVelocity(Velocity=&RandomRangeVector(Minimum=<0 -20 50>, Maximum=<0 20 150>))
  ]
  particleUpdateScript [
   &ParticleState()
   &ScaleMeshSize(Scale Factor=&VectorFromFloat(Value=&FloatFromCurve(FloatCurve=DataInterface() {propertyValues {Curve {keys [<at 0 0.8>,<at 0.2 1.25>,<at 1 0>]}}})))
   &Drag(Use Linear Drag=true, Use Rotational Drag=true)
   &FN_Sprite_Attributes(
    Color=&MakeLinearColorFromVectorAndFloat(
     "Vector (RGB)"=&MakeVectorFromLinearColorRGB(Color=<rgba 1.5 0.55 0.075 1>)
     "Float (Alpha)"=&FloatFromCurve(FloatCurve=DataInterface() {propertyValues {Curve {keys [<at 0.5 1>,<at 1 0>]}}}, Scale Curve=0.1)
    )
    Create Color Attribute=true
    Create SpriteSize Attribute=true
    Module Output:ENiagara_AttributeSamplingApplyOutput=::NewEnumerator1
    SpriteSize=&Multiply_Vector2D(A=$Particles.SpriteSize, B=<9 6>)
   )
   &CurlNoiseForce(Bias Noise=false, Noise Strength=150, Pan Noise=false)
   &SolveForcesAndVelocity(Clamp Velocity=false, Limit Acceleration=false)
   &KillParticlesInVolume(
    Box Size:Vector3f=<150 80 100>
    Kill Shape=::NewEnumerator1
    Origin Offset:Vector3f=&MakeVector(
     X=&FloatFromCurve(FloatCurve=DataInterface() {propertyValues {Curve {keys [<at 0 0>,<at 4 -3>]}}}, CurveIndex=$Emitter.Age, Scale Curve=50)
     Y=0
     Z=0
    )
   )
  ]
  rendererClasses [&NiagaraMeshRendererProperties, &NiagaraSpriteRendererProperties]
  renderers [
   {rendererClass=&NiagaraMeshRendererProperties, rendererIndex=0}
   {rendererClass=&NiagaraSpriteRendererProperties, rendererIndex=1}
  ]
 }
 energy {
  emitterUpdateScript [
   &EmitterState(Life Cycle Mode=::NewEnumerator1, Loop Behavior=::NewEnumerator1, Loop Delay=1, Loop Duration=4, UseLoopDelay=true)
   &SpawnRate(SpawnRate=20, Use Spawn Probability=false)
  ]
  particleSpawnScript [
   &InitializeParticle(
    Color=<rgba 3 0.7764 0.06 1>
    Lifetime Max=1.5
    Lifetime Mode=::NewEnumerator1
    Sprite Size Max=<40 110>
    Sprite Size Min=<30 70>
    Sprite Size Mode=::NewEnumerator2
    Sprite UV Mode=::NewEnumerator1
    UsePositionOffset=false
    Write Material Random=true
   )
   &AddVelocity(Velocity=<0 0 5>)
  ]
  particleUpdateScript [
   &ParticleState()
   &ScaleSpriteSize(
    Scale Factor=&RandomRangeVector2D(Minimum=<0.55 0.65>, Maximum=<0.55 0.65>)
    Scale Sprite Size Mode=::NewEnumerator2
   ) {module=ScaleSpriteSize001}
   &ScaleColor(
    Scale Alpha=&FloatFromCurve(FloatCurve=DataInterface() {propertyValues {Curve {keys [
         {tangentMode=RCTM_User, time=0, value=0}
         <at 0.2 1>
         {arriveTangent=-0.0405, interpMode=RCIM_Cubic, leaveTangent=-0.0405, tangentMode=RCTM_User, time=0.4, value=1}
         {arriveTangent=-1.4489, interpMode=RCIM_Cubic, leaveTangent=-1.4489, tangentMode=RCTM_User, time=1, value=0}
        ]}}})
    Scale RGB=&VectorFromFloat(Value=&FNTimeOfDayScalar(Night Scale=0.75, Morning Scale=1, Evening Scale=1))
    ScaleA=true
    ScaleRGB=true
   )
   &ShapeLocation(
    Box Size=<90 5 60>
    MaskSpawnGroup=false
    Offset=<0 0 10>
    Offset Mode=::NewEnumerator0
    Override Seed=false
    Shape Primitive=::NewEnumerator2
    Surface Only=false
    UseOwnerScale=true
   )
   &SolveForcesAndVelocity(Clamp Velocity=false, Limit Acceleration=false)
   &KillParticlesInVolume(
    Box Size:Vector3f=<150 80 100>
    Kill Shape=::NewEnumerator1
    Origin Offset:Vector3f=&MakeVector(
     X=&FloatFromCurve(FloatCurve=DataInterface() {propertyValues {Curve {keys [<at 0 0>,<at 3 -2.5>]}}}, CurveIndex=$Emitter.Age, Scale Curve=50)
     Y=0
     Z=0
    )
   )
  ]
  rendererClasses [&NiagaraSpriteRendererProperties]
  renderers [{rendererClass=&NiagaraSpriteRendererProperties}]
 }
 flare {
  emitterUpdateScript [
   &EmitterState(Life Cycle Mode=::NewEnumerator1, Loop Behavior=::NewEnumerator1, Loop Delay=0.8, Loop Duration=4.5, UseLoopDelay=true)
   &SpawnRate(SpawnRate=20, Use Spawn Probability=false)
  ]
  particleSpawnScript [
   &InitializeParticle(
    Color=<rgba 1 0.4 0.1 1>
    Lifetime Max=0.8
    Lifetime Min=0.4
    Lifetime Mode=::NewEnumerator1
    Position Offset=<130 0 0>
    Sprite Size Max=<1.5 4>
    Sprite Size Min=<1 3>
    Sprite Size Mode=::NewEnumerator2
    UsePositionOffset=true
    Write Material Random=true
   )
   &ShapeLocation(
    Box Midpoint=<1 0.5 0.5>
    Box Size=&Multiply_Vector(A=<100 10 50>, B=$System.RangeSize)
    MaskSpawnGroup=false
    Non Uniform Scale=&MakeVector(
     X=&FloatFromCurve(
      FloatCurve=DataInterface() {propertyValues {Curve {
         defaultValue=1
         keys [
          {arriveTangent=0.093, interpMode=RCIM_Cubic, leaveTangent=0.2308, tangentMode=RCTM_Break, time=0, value=0}
          {arriveTangent=0.1562, interpMode=RCIM_Cubic, tangentMode=RCTM_Break, time=5, value=1}
         ]
        }}}
      CurveIndex=$Emitter.Age
     )
     Y=1
     Z=1
    )
    Offset=<-80 0 0>
    Offset Mode=::NewEnumerator0
    Override Seed=false
    Shape Primitive=::NewEnumerator2
    Surface Only=false
    UseOwnerScale=true
   )
   &AddVelocity(Velocity=&RandomRangeVector(Minimum=<0 -15 50>, Maximum=<0 15 150>))
  ]
  particleUpdateScript [
   &ParticleState()
   &Drag(Use Linear Drag=true, Use Rotational Drag=true)
   &CurlNoiseForce(Bias Noise=false, Noise Strength=260, Pan Noise=false)
   &SolveForcesAndVelocity(Clamp Velocity=false, Limit Acceleration=false)
   &FN_Sprite_Attributes(
    Color=&MakeLinearColorFromVectorAndFloat(
     "Vector (RGB)"=&MakeVectorFromLinearColorRGB(Color=<rgba 1.5 0.55 0.075 1>)
     "Float (Alpha)"=&FloatFromCurve(FloatCurve=DataInterface() {propertyValues {Curve {keys [<at 0.5 1>,<at 1 0>]}}}, Scale Curve=0.1)
    )
    Create Color Attribute=true
    Create SpriteSize Attribute=true
    Module Output:ENiagara_AttributeSamplingApplyOutput=::NewEnumerator1
    SpriteSize=&Multiply_Vector2D(A=$Particles.SpriteSize, B=<8 6>)
   )
  ]
  rendererClasses [&NiagaraSpriteRendererProperties]
  renderers [
   {rendererClass=&NiagaraSpriteRendererProperties}
   {rendererClass=&NiagaraSpriteRendererProperties, rendererIndex=1}
  ]
 }
}
summary {
 systemName=NS_Emote_LemonWrath_Wall
 emitters [
  {emitterName=flare, rendererClasses [&NiagaraSpriteRendererProperties]}
  {emitterName=Main, rendererClasses [&NiagaraMeshRendererProperties]}
  {emitterName=energy, rendererClasses [&NiagaraSpriteRendererProperties]}
  {emitterName=box, rendererClasses [&NiagaraMeshRendererProperties, &NiagaraSpriteRendererProperties]}
  {emitterName=Decal_cracksmall, rendererClasses [&NiagaraDecalRendererProperties]}
  {emitterName=Decal_crack001, rendererClasses [&NiagaraDecalRendererProperties, &NiagaraLightRendererProperties]}
 ]
 userVariables [
  User.bIsFrontEnd {type=NiagaraBool}
  User.bIsFrontEndPreview {type=NiagaraBool}
  User.RibbonIDOffset {type=NiagaraInt32}
  User.SwingSectionIndex {type=NiagaraInt32}
 ]
}
