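# Niagara system that draws a short-lived, orange weapon trail for the LemonWrath emote.
# A single FN_AnimTrail ribbon emitter uses custom anim trail modules to spawn, initialize bone indices and update particles, with a socket reader among its dependencies.
# Curves fade alpha over each particle's life and taper ribbon width near the end.
# User parameters expose front-end flags, ribbon ID offset, swing section index and variant color.
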NiagaraSystem /BRCosmetics/Effects/Fort_Effects/Emotes/LemonWrath/NS_Emote_LemonWrath_Weapon
dependencies {
 usedDataInterfaces [&NiagaraDataInterfaceCurve, &NiagaraDataInterfaceSocketReader]
 usedDynamicInputs []
 usedModules [
  &self:SystemSpawnScript.SystemScriptSource.SystemScriptGraph.NiagaraNodeAssignment_0.SetVariables_#1
  &SystemState
 ]
}
emitters {FN_AnimTrail {
  emitterUpdateScript [
   &EmitterState()
   &FN_SpawnAnimTrail()
  ]
  particleSpawnScript [
   &FN_InitializeAnimTrailBoneIndecies()
   &FN_InitializeAnimTrail(Color=<rgba 1 0.25 0.1 0.5>, Lifetime=0.16)
  ]
  particleUpdateScript [
   &SolveForcesAndVelocity(Clamp Velocity=false, Limit Acceleration=false)
   &FN_UpdateAnimTrail() {module=FN_UpdateAnimTrail001}
   &ScaleColor(
    Scale Alpha=&FloatFromCurve(FloatCurve=DataInterface() {propertyValues {Curve {keys [<at 0 1>,<at 1 0>]}}})
    ScaleA=true
    ScaleRGB=true
   )
   &ScaleRibbonWidth(Ribbon Width Scale=&FloatFromCurve(FloatCurve=DataInterface() {propertyValues {Curve {
        defaultValue=1
        keys [
         {tangentMode=RCTM_User, time=0, value=1}
         {arriveTangent=-0.0405, interpMode=RCIM_Cubic, leaveTangent=-0.0405, tangentMode=RCTM_User, time=0.7009, value=1}
         {arriveTangent=-8.0089, interpMode=RCIM_Cubic, leaveTangent=-8.0089, tangentMode=RCTM_User, time=1, value=0}
        ]
       }}}))
   &FN_Sprite_Attributes(Color=<rgba 1 0.35 0.08 1>, Create Color Attribute=true, Module Output:ENiagara_AttributeSamplingApplyOutput=::NewEnumerator1)
  ]
  rendererClasses [&NiagaraRibbonRendererProperties]
  renderers [{rendererClass=&NiagaraRibbonRendererProperties}]
 }}
summary {
 systemName=NS_Emote_LemonWrath_Weapon
 emitters [{emitterName=FN_AnimTrail, rendererClasses [&NiagaraRibbonRendererProperties]}]
 userVariables [
  User.bIsFrontEnd {type=NiagaraBool}
  User.bIsFrontEndPreview {type=NiagaraBool}
  User.RibbonIDOffset {type=NiagaraInt32}
  User.SwingSectionIndex {type=NiagaraInt32}
  User.VariantColor {type=NiagaraFloat, defaultValue=0, description=""}
 ]
}
