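# Niagara system for the LemonWrath emote that places a fissure crack decal, with spawning suppressed in front-end preview.
# A single burst emitter creates a long-lived decal whose fade curve snaps in quickly and drops out at the end.
# Each update samples the skeletal mesh, filtered to the root bone, into Position02 and MeshOrientation.
# Both lerps sit at zero, so these attributes take the sampled transform rather than the spawn values.
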NiagaraSystem /BRCosmetics/Effects/Fort_Effects/Emotes/LemonWrath/NS_Emote_LemonWrath_fissures
dependencies {
 usedDataInterfaces [&NiagaraDataInterfaceSkeletalMesh, &NiagaraDataInterfaceCurve]
 usedDynamicInputs []
 usedModules [&SystemState]
}
emitters {Decal_crack001 {
  emitterUpdateScript [
   &EmitterState(Life Cycle Mode=::NewEnumerator1, Loop Behavior=::NewEnumerator1, Loop Duration=30, UseLoopDelay=false)
   &SpawnBurst_Instantaneous(
    Spawn Probability=&MakeCustomFloatFromBool(Bool=$User.bIsFrontEndPreview, True Float=0, False Float=1)
    Use Spawn Probability=true
    UseLoopCountLimit=false
   )
  ]
  particleSpawnScript [
   &InitializeParticle(Color=<rgba 0.1 0.018 0.003 1>, Lifetime=30, UsePositionOffset=true, Write Material Random=true)
   /Niagara/Modules/Spawn/Decal/InitialDecalOrientation(Decal Orientation Mode:ENiagara_MeshOrientationOptions=::NewEnumerator1, Use Rotation Vector=true)
  ]
  particleUpdateScript [
   &SkeletalMeshLocation(
    Bone Sampling Mode=::NewEnumerator4
    MaskSpawnGroup=false
    Orientation Sampling=::NewEnumerator1
    Position Sampling=::NewEnumerator2
    Skeletal Mesh=DataInterface() {propertyValues {
      DefaultMesh=null
      ExcludeBoneName=root
      OptionalComponentClass=null
      PreviewMesh=/BRCosmetics/Characters/Player/Female/Medium/Bodies/F_MED_LemonWrath_Ochre/Meshes/F_MED_LemonWrath_Ochre
      SkinningMode=SkinOnTheFly
      SoftSourceActor=null
      SourceMode=Default
      UvSetIndex=0
      WholeMeshLOD=-1
      bExcludeBone=false
      bReadDeformedGeometry=true
      bRequireCurrentFrameData=true
      ComponentTags []
      FilteredBones [root]
      FilteredSockets []
      MeshUserParameter {parameter {name=None, typeDefHandle {registeredTypeIndex=1}, varData []}}
      SamplingRegions []
     }}
   )
   /Niagara/Modules/Decal/Decal_Attributes(
    Color=1
    Decal Fade:NiagaraFloat=&FloatFromCurve(FloatCurve=DataInterface() {propertyValues {Curve {keys [<at 0 0.5>,<at 0.01 1>,<at 0.9485 1.0042>,<at 1 0>]}}})
    Decal Size:Vector3f=<70 110 50>
    Fade=1
    Scale Alpha=-1
    Scale RGB:Vector3f=&VectorFromFloat(Value=1)
    ScaleA=true
    ScaleRGB=true
    Size=1
   )
   &ParticleState()
   /SetVariables(Particles.Position02:NiagaraPosition=&Lerp_Vector(
     A=Linked() {linkedVariable {name=Particles.SkeletalMeshLocation.SampledPosition, type=NiagaraPosition}}
     B=Linked() {linkedVariable {name=Particles.Position, type=NiagaraPosition}}
     Alpha=0
    )) {isSetParameters=true}
   /SetVariables(Particles.MeshOrientation=&LerpQuaternion(
     Quaternion A:Quat4f=$Particles.SkeletalMeshLocation.SampledRotation
     Quaternion B:Quat4f=$Particles.DecalOrientation
     Lerp Factor:NiagaraFloat=0
     Enable Spherical Linear Interpolation:NiagaraBool=false
    )) {isSetParameters=true}
  ]
  rendererClasses [&NiagaraDecalRendererProperties]
  renderers [{rendererClass=&NiagaraDecalRendererProperties, rendererIndex=0}]
 }}
summary {
 systemName=NS_Emote_LemonWrath_fissures
 emitters [{emitterName=Decal_crack001, rendererClasses [&NiagaraDecalRendererProperties]}]
 userVariables [
  User.bIsFrontEnd {type=NiagaraBool}
  User.bIsFrontEndPreview {type=NiagaraBool}
 ]
}
